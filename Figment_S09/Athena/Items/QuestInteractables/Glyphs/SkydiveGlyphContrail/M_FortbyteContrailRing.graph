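# Masked, lit mesh-particle material for the Fortbyte skydive contrail ring, built on a Fortbyte gradient texture scaled by an Intensity parameter.
# Gradient UVs pan and pass through CustomRotator, with an added distortion term that the Low quality switch drops.
# The rotated UVs also tile a swirl-flow normal map.
# An IsTriangleMesh static switch picks the UV angle and one of two position-offset branches, each adding a BobHeight blend to RotateAboutAxis.
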Material /Figment_S09/Athena/Items/QuestInteractables/Glyphs/SkydiveGlyphContrail/M_FortbyteContrailRing
graph {
 properties {bUsedWithMeshParticles=true, blendMode=BLEND_Masked, pixelDepthOffsetMode=PDOM_Legacy, usedShadingModels=Default Lit}
 nodes [
  Constant() {r=-0.01}
  Constant() {r=1}
  Constant() {r=0.01}
  Constant() {r=1}
  Constant() {r=0.01}
  Constant() {r=-0.01}
  n1 = ObjectPositionWS()
  n2 = WorldPosition()
  MaterialFunctionCall!(Rotation Amount, PivotPoint=@n1, WorldPosition=@n2.XYZ) {materialFunction=/Engine/Functions/Engine_MaterialFunctions02/WorldPositionOffset/RotateAboutWorldAxis_cheap}
  n3 = Multiply(A=ScalarParameter() {defaultValue=0.3, parameterName=UVDistort_Mult})
  n4 = Panner(Coordinate=TextureCoordinate() {vTiling=0.5}) {speedY=0.25}
  n5 = MaterialFunctionCall(
   UVs=QualitySwitch(Default=Add(A=@n3, B=@n4), Low=@n4)
   "Rotation Angle (0-1)"=StaticSwitchParameter(
    True=ScalarParameter() {defaultValue=0.5, parameterName=RotAngle}
    False=ScalarParameter() {defaultValue=0.25, parameterName="Rotation Angle (0-1) (S)"}
   ) {parameterName=IsTriangleMesh}
  ) {materialFunction=&CustomRotator}
  Multiply(
   A=ScalarParameter() {defaultValue=1, parameterName=Intensity, sliderMax=30, sliderMin=1}
   B=TextureSample(UVs=@n5.Rotated Values) {texture=/Game/Athena/UI/Challenges/Art/FortByte/T_UI_ChallengeTile_Fortbytes_Grad}.RGB
  )
  n6 = CameraPositionWS()
  n7 = WorldPosition()
  Reroute(None=RotateAboutAxis(
    NormalizedRotationAxis=Constant3Vector() {constant=<rgba 1 0 0 0>}
    RotationAngle=LinearInterpolate(A=@n6, B=@n7.XYZ)
    PivotPoint=ObjectPositionWS()
    Position=WorldPosition().XYZ
   ))
  ScalarParameter() {defaultValue=0.4}
  n8 = VectorParameter() {parameterName=BobHeightMax, defaultValue=<rgba 0 0 5 0>}
  n9 = VectorParameter() {parameterName=BobHeightMin}
  n10 = Constant3Vector() {constant=<rgba 0 0 1 0>}
  n11 = LinearInterpolate(A=Constant() {r=-0.01}, B=Constant() {r=0.01})
  n12 = VectorParameter() {parameterName=BobHeightMax, defaultValue=<rgba 0 0 5 0>}
  n13 = VectorParameter() {parameterName=BobHeightMin}
  n14 = Constant3Vector() {constant=<rgba 1 1 0 0>}
  n15 = LinearInterpolate(A=CameraPositionWS(), B=WorldPosition().XYZ)
  n16 = ObjectPositionWS()
  n17 = WorldPosition()
  StaticSwitchParameter(
   True=Add(
    A=LinearInterpolate(A=@n8.RGB, B=@n9.RGB)
    B=RotateAboutAxis(NormalizedRotationAxis=@n10, RotationAngle=@n11, PivotPoint=@n1, Position=@n2.XYZ)
   )
   False=Add(
    A=LinearInterpolate(A=@n12.RGB, B=@n13.RGB)
    B=RotateAboutAxis(NormalizedRotationAxis=@n14, RotationAngle=@n15, PivotPoint=@n16, Position=@n17.XYZ)
   )
  ) {parameterName=IsTriangleMesh}
  TextureSample(UVs=Multiply(B=@n5.Rotated Values) {constA=2}) {samplerType=SAMPLERTYPE_Normal, texture=/Figment_S09/Athena/Environments/Nexus/Textures/T_SwirlFlow_01_N}
 ]
}
